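# Anim Blueprint for the RankedTandem pickaxe that adds procedural secondary motion to its straw, glasses and popcorn bones.
# The cached input pose feeds two differently tuned Anim Dynamics chains, with additive bone translations pre-offsetting popcorn 02–07 before simulation.
# bIsInFrontEnd selects between them: the front-end chain lets popcorn rotate and shift on several axes, while the other locks popcorn rotation and springs it along one axis.
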AnimBlueprint /BRCosmetics/Weapons/FORT_Melee/Pickaxe_RankedTandem/Meshes/RankedTandem_Axe_AnimBP
target_skeleton=&/RankedTandem_Axe_Skeleton
graphs [
 AnimGraph {nodes [
   n1 = Anim Dynamics - Bone: dyn_straw01_base_rot(ComponentPose=Local To Component(LocalPose=Use cached pose 'input pose'()), Alpha=1, ExternalForce="0,0,0") {
    angularDampingOverride=1
    angularSpringConstant=5000
    bAngularSpring=true
    bOverrideAngularDamping=true
    bUseGravityOverride=true
    linearDampingOverride=0.9
    linearSpringConstant=30
    boundBone {boneName=dyn_straw01_base_rot}
    componentAppliedLinearAccClamp=<900 900 740>
    componentLinearAccScale=<0.5 0.5 0.5>
    componentLinearVelScale=<0.6 0.6 0.6>
    gravityOverride {z=-980}
    physicsBodyDefinitions [{
      boundBone {boneName=dyn_straw01_base_rot}
      boxExtents=<20 20 20>
      constraintSetup {angularTargetAxis=AxisY, twistAxis=AxisY, angularLimitsMax=<50 0 0>, angularLimitsMin=<-50 0 0>, angularTarget=<0 1 0>}
      localJointOffset=<0 1.2 0>
     }]
    simSpaceSettings {maxAngularAcceleration=90, maxAngularVelocity=6, simSpaceAngularAlpha=0.7}
   }
   n2 = Anim Dynamics - Bone: dyn_straw01_bend(ComponentPose=@n1, Alpha=1, ExternalForce="0,0,0") {
    angularDampingOverride=1
    angularSpringConstant=5000
    bAngularSpring=true
    bOverrideAngularDamping=true
    bUseGravityOverride=true
    linearDampingOverride=0.9
    linearSpringConstant=30
    boundBone {boneName=dyn_straw01_bend}
    componentAppliedLinearAccClamp=<1000 1000 1000>
    componentLinearAccScale=<0.4 0.4 0.4>
    componentLinearVelScale=<0.5 0.5 0.5>
    gravityOverride {z=-980}
    physicsBodyDefinitions [{
      boundBone {boneName=dyn_straw01_bend}
      boxExtents=<18 18 18>
      constraintSetup {angularLimitsMax=<0 40 0>, angularLimitsMin=<0 -20 0>, angularTarget=<1 0 0>}
      localJointOffset=<1 0.3 0>
     }]
    simSpaceSettings {maxAngularAcceleration=90, maxAngularVelocity=6, simSpaceAngularAlpha=0.6}
   }
   n3 = Anim Dynamics - Bone: dyn_straw02_base_rot(ComponentPose=@n2, Alpha=1, ExternalForce="0,0,0") {
    angularDampingOverride=1
    angularSpringConstant=5000
    bAngularSpring=true
    bOverrideAngularDamping=true
    bUseGravityOverride=true
    linearDampingOverride=0.9
    linearSpringConstant=30
    boundBone {boneName=dyn_straw02_base_rot}
    componentAppliedLinearAccClamp=<900 900 780>
    componentLinearAccScale=<0.5 0.5 0.4>
    componentLinearVelScale=<0.6 0.6 0.5>
    gravityOverride {z=-980}
    physicsBodyDefinitions [{
      boundBone {boneName=dyn_straw02_base_rot}
      boxExtents=<20 20 18>
      constraintSetup {angularTargetAxis=AxisY, twistAxis=AxisY, angularLimitsMax=<40 0 0>, angularLimitsMin=<-40 0 0>, angularTarget=<0 1 0>}
      localJointOffset=<0.2 -1.3 0.2>
     }]
    simSpaceSettings {maxAngularAcceleration=110, maxAngularVelocity=8, simSpaceAngularAlpha=0.7}
   }
   n4 = Anim Dynamics - Bone: dyn_straw02_bend(ComponentPose=@n3, Alpha=1, ExternalForce="0,0,0") {
    angularDampingOverride=1
    angularSpringConstant=5000
    bAngularSpring=true
    bOverrideAngularDamping=true
    bUseGravityOverride=true
    linearDampingOverride=0.9
    linearSpringConstant=30
    boundBone {boneName=dyn_straw02_bend}
    componentAppliedLinearAccClamp=<1000 1000 1000>
    componentLinearAccScale=<0.4 0.5 0.6>
    componentLinearVelScale=<0.5 0.6 0.6>
    gravityOverride {z=-980}
    physicsBodyDefinitions [{
      boundBone {boneName=dyn_straw02_bend}
      boxExtents=<25 10 15>
      constraintSetup {angularLimitsMax=<0 20 0>, angularLimitsMin=<0 -40 0>, angularTarget=<1 0 0>}
      localJointOffset=<0.9 0.1 -1>
     }]
    simSpaceSettings {maxAngularAcceleration=110, maxAngularVelocity=8, simSpaceAngularAlpha=0.7}
   }
   n5 = Anim Dynamics - Bone: dyn_glasses(ComponentPose=@n4, Alpha=1, ExternalForce="0,0,0") {
    angularDampingOverride=0.9999
    angularSpringConstant=3000
    bAngularSpring=true
    bOverrideAngularDamping=true
    bUseGravityOverride=true
    boundBone {boneName=dyn_glasses}
    componentAppliedLinearAccClamp=<1000 1000 1000>
    componentLinearAccScale=<0.4 0.4 0.4>
    componentLinearVelScale=<0.5 0.5 0.5>
    gravityOverride {z=-900}
    physicsBodyDefinitions [{
      boundBone {boneName=dyn_glasses}
      boxExtents=<30 30 30>
      constraintSetup {angularLimitsMax=<0 2 0>, angularLimitsMin=<0 -2 0>, angularTarget=<1 0 0>}
      localJointOffset=<0.7 0 0>
     }]
    simSpaceSettings {maxAngularAcceleration=100, maxAngularVelocity=10, simSpaceAngularAlpha=0.7}
   }
   n6 = Anim Dynamics - Bone: dyn_popcorn_01!(ComponentPose=@n5, bAlphaBoolEnabled:bool=True, ExternalForce="0,0,0") -> (Pose) {
    alphaInputType=Bool
    angularDampingOverride=0.9
    angularSpringConstant=1200
    bAngularSpring=true
    bLinearSpring=true
    bOverrideAngularDamping=true
    bOverrideLinearDamping=true
    bUseGravityOverride=true
    linearDampingOverride=0.99
    linearSpringConstant=85
    alphaBoolBlend {blendOption=Linear}
    alphaScaleBiasClamp {bClampResult=true}
    boundBone {boneName=dyn_popcorn_01}
    componentAppliedLinearAccClamp=<500 500 800>
    componentLinearAccScale=<0.2 0.2 0.4>
    componentLinearVelScale=<0.3 0.3 0.5>
    gravityOverride {z=-980}
    physicsBodyDefinitions [{
      boundBone {boneName=dyn_popcorn_01}
      boxExtents=<20 20 15>
      constraintSetup {angularLimitsMax=<15 5 5>, angularLimitsMin=<-5 -5 -15>, angularTarget=<0 0 0>, linearAxesMax=<0.1 0.5 0.5>}
      localJointOffset=<0.4 0.3 0.1>
     }]
    simSpaceSettings {maxAngularAcceleration=90, maxAngularVelocity=25, simSpaceAngularAlpha=1}
   }
   n7 = "Transform (Modify) Bone - Bone: dyn_popcorn_02"(ComponentPose=@n6, Alpha=1, Translation="-4.797,-0.000,0.000", Rotation="0,0,0", Scale="1,1,1") {translationMode=BMM_Additive, translationSpace=BCS_BoneSpace, boneToModify {boneName=dyn_popcorn_02}, translation {x=-4.797}}
   n8 = Anim Dynamics - Bone: dyn_popcorn_02!(ComponentPose=@n7, bAlphaBoolEnabled:bool=True, ExternalForce="0,0,0") -> (Pose) {
    alphaInputType=Bool
    angularDampingOverride=0.99
    angularSpringConstant=700
    bAngularSpring=true
    bLinearSpring=true
    bOverrideAngularDamping=true
    bOverrideLinearDamping=true
    bUseGravityOverride=true
    linearDampingOverride=0.99
    linearSpringConstant=70
    alphaBoolBlend {blendOption=Linear}
    alphaScaleBiasClamp {bClampResult=true}
    boundBone {boneName=dyn_popcorn_02}
    componentAppliedLinearAccClamp=<500 400 700>
    componentLinearAccScale=<0.2 0.2 0.5>
    componentLinearVelScale=<0.3 0.3 0.6>
    gravityOverride {z=-980}
    physicsBodyDefinitions [{
      boundBone {boneName=dyn_popcorn_02}
      boxExtents=<5 30 20>
      constraintSetup {angularLimitsMax=<5 15 20>, angularLimitsMin=<-5 -15 -5>, angularTarget=<0 0 0>, linearAxesMax=<0.1 0.2 0.1>}
      localJointOffset=<0.5 0.1 0.1>
     }]
    simSpaceSettings {maxAngularAcceleration=100, maxAngularVelocity=30, simSpaceAngularAlpha=1}
   }
   n9 = "Transform (Modify) Bone - Bone: dyn_popcorn_03"(ComponentPose=@n8, Alpha=1, Translation="-4.797,-0.000,0.000", Rotation="0,0,0", Scale="1,1,1") {translationMode=BMM_Additive, translationSpace=BCS_BoneSpace, boneToModify {boneName=dyn_popcorn_03}, translation {x=-4.797}}
   n10 = Anim Dynamics - Bone: dyn_popcorn_03!(ComponentPose=@n9, bAlphaBoolEnabled:bool=True, ExternalForce="0,0,0") -> (Pose) {
    alphaInputType=Bool
    angularDampingOverride=0.9
    angularSpringConstant=1200
    bAngularSpring=true
    bLinearSpring=true
    bOverrideAngularDamping=true
    bOverrideLinearDamping=true
    bUseGravityOverride=true
    linearDampingOverride=0.99
    linearSpringConstant=100
    alphaBoolBlend {blendOption=Linear}
    alphaScaleBiasClamp {bClampResult=true}
    boundBone {boneName=dyn_popcorn_03}
    componentAppliedLinearAccClamp=<500 500 600>
    componentLinearAccScale=<0.2 0.2 0.5>
    componentLinearVelScale=<0.3 0.3 0.6>
    gravityOverride {z=-980}
    physicsBodyDefinitions [{
      boundBone {boneName=dyn_popcorn_03}
      boxExtents=<5 5 5>
      constraintSetup {angularLimitsMax=<5 5 20>, angularLimitsMin=<-5 -15 -5>, angularTarget=<0 0 0>, linearAxesMax=<0.1 0.2 0.5>}
      localJointOffset=<0.5 0.1 0.2>
     }]
    simSpaceSettings {maxAngularAcceleration=100, maxAngularVelocity=30, simSpaceAngularAlpha=1}
   }
   n11 = "Transform (Modify) Bone - Bone: dyn_popcorn_04"(ComponentPose=@n10, Alpha=1, Translation="-3.548,0.000,-0.000", Rotation="0,0,0", Scale="1,1,1") {translationMode=BMM_Additive, translationSpace=BCS_BoneSpace, boneToModify {boneName=dyn_popcorn_04}, translation {x=-3.548}}
   n12 = Anim Dynamics - Bone: dyn_popcorn_04!(ComponentPose=@n11, bAlphaBoolEnabled:bool=True, ExternalForce="0,0,0") -> (Pose) {
    alphaInputType=Bool
    angularDampingOverride=0.99
    angularSpringConstant=1000
    bAngularSpring=true
    bLinearSpring=true
    bOverrideAngularDamping=true
    bOverrideLinearDamping=true
    bUseGravityOverride=true
    linearDampingOverride=0.99
    linearSpringConstant=50
    alphaBoolBlend {blendOption=Linear}
    alphaScaleBiasClamp {bClampResult=true}
    boundBone {boneName=dyn_popcorn_04}
    componentAppliedLinearAccClamp=<500 400 700>
    componentLinearAccScale=<0.2 0.2 0.5>
    componentLinearVelScale=<0.3 0.3 0.6>
    gravityOverride {z=-980}
    physicsBodyDefinitions [{
      boundBone {boneName=dyn_popcorn_04}
      boxExtents=<20 20 7>
      constraintSetup {angularLimitsMax=<5 10 20>, angularLimitsMin=<-5 -15 -5>, angularTarget=<0 0 0>, linearAxesMax=<0.1 0.5 0.5>}
      localJointOffset=<0.5 0.2 0.1>
     }]
    simSpaceSettings {maxAngularAcceleration=100, maxAngularVelocity=30, simSpaceAngularAlpha=0.7}
   }
   n13 = "Transform (Modify) Bone - Bone: dyn_popcorn_05"(ComponentPose=@n12, Alpha=1, Translation="-1.897,0.000,0.753", Rotation="0,0,0", Scale="1,1,1") {translationMode=BMM_Additive, translationSpace=BCS_BoneSpace, boneToModify {boneName=dyn_popcorn_05}, translation {x=-1.897, z=0.753}}
   n14 = Anim Dynamics - Bone: dyn_popcorn_05!(ComponentPose=@n13, bAlphaBoolEnabled:bool=True, ExternalForce="0,0,0") -> (Pose) {
    alphaInputType=Bool
    angularDampingOverride=0.9
    angularSpringConstant=700
    bLinearSpring=true
    bOverrideAngularDamping=true
    bOverrideLinearDamping=true
    bUseGravityOverride=true
    linearDampingOverride=0.99
    linearSpringConstant=80
    alphaBoolBlend {blendOption=Linear}
    alphaScaleBiasClamp {bClampResult=true}
    boundBone {boneName=dyn_popcorn_05}
    componentAppliedLinearAccClamp=<500 500 600>
    componentLinearAccScale=<0.2 0.2 0.4>
    componentLinearVelScale=<0.3 0.3 0.5>
    gravityOverride {z=-980}
    physicsBodyDefinitions [{
      boundBone {boneName=dyn_popcorn_05}
      boxExtents=<20 30 20>
      constraintSetup {angularLimitsMax=<15 15 5>, angularLimitsMin=<-5 -5 -15>, angularTarget=<0 0 0>, linearAxesMax=<0.1 0.1 0.1>}
      localJointOffset=<0.3 0.3 0.1>
     }]
    simSpaceSettings {maxAngularAcceleration=90, maxAngularVelocity=25, simSpaceAngularAlpha=1}
   }
   n15 = "Transform (Modify) Bone - Bone: dyn_popcorn_06"(ComponentPose=@n14, Alpha=1, Translation="-3.548,-1.146,-0.000", Rotation="0,0,0", Scale="1,1,1") {translationMode=BMM_Additive, translationSpace=BCS_BoneSpace, boneToModify {boneName=dyn_popcorn_06}, translation=<-3.548 -1.146>}
   n16 = Anim Dynamics - Bone: dyn_popcorn_06!(ComponentPose=@n15, bAlphaBoolEnabled:bool=True, ExternalForce="0,0,0") -> (Pose) {
    alphaInputType=Bool
    angularDampingOverride=0.9
    angularSpringConstant=700
    bLinearSpring=true
    bOverrideAngularDamping=true
    bOverrideLinearDamping=true
    bUseGravityOverride=true
    linearDampingOverride=0.99
    linearSpringConstant=100
    alphaBoolBlend {blendOption=Linear}
    alphaScaleBiasClamp {bClampResult=true}
    boundBone {boneName=dyn_popcorn_06}
    componentAppliedLinearAccClamp=<500 500 730>
    componentLinearAccScale=<0.2 0.2 0.5>
    componentLinearVelScale=<0.3 0.3 0.6>
    gravityOverride {z=-980}
    physicsBodyDefinitions [{
      boundBone {boneName=dyn_popcorn_06}
      boxExtents=<20 30 20>
      constraintSetup {angularLimitsMax=<15 15 5>, angularLimitsMin=<-5 -5 -15>, angularTarget=<0 0 0>, linearAxesMax=<0.1 0.2 0.2>}
      localJointOffset=<0.5 -0.5 0.3>
     }]
    simSpaceSettings {maxAngularAcceleration=90, maxAngularVelocity=25, simSpaceAngularAlpha=1}
   }
   n17 = "Transform (Modify) Bone - Bone: dyn_popcorn_07"(ComponentPose=@n16, Alpha=1, Translation="-1.084,-1.146,-0.000", Rotation="0,0,0", Scale="1,1,1") {translationMode=BMM_Additive, translationSpace=BCS_BoneSpace, boneToModify {boneName=dyn_popcorn_07}, translation=<-1.084 -1.146>}
   n18 = Anim Dynamics - Bone: dyn_popcorn_07!(ComponentPose=@n17, bAlphaBoolEnabled:bool=True, ExternalForce="0,0,0") -> (Pose) {
    alphaInputType=Bool
    angularDampingOverride=0.99
    angularSpringConstant=1200
    bAngularSpring=true
    bLinearSpring=true
    bOverrideAngularDamping=true
    bOverrideLinearDamping=true
    bUseGravityOverride=true
    linearDampingOverride=0.99
    linearSpringConstant=70
    alphaBoolBlend {blendOption=Linear}
    alphaScaleBiasClamp {bClampResult=true}
    boundBone {boneName=dyn_popcorn_07}
    componentAppliedLinearAccClamp=<500 400 700>
    componentLinearAccScale=<0.2 0.2 0.4>
    componentLinearVelScale=<0.3 0.3 0.5>
    gravityOverride {z=-980}
    physicsBodyDefinitions [{
      boundBone {boneName=dyn_popcorn_07}
      boxExtents=<3 30 20>
      constraintSetup {angularLimitsMax=<5 5 5>, angularLimitsMin=<-5 -5 -5>, angularTarget=<0 0 0>, linearAxesMax=<0.1 0.4 0.25>}
      localJointOffset=<0.6 0 0>
     }]
    simSpaceSettings {maxAngularAcceleration=100, maxAngularVelocity=30, simSpaceAngularAlpha=1}
   }
   n19 = Anim Dynamics - Bone: dyn_straw01_base_rot(ComponentPose=Local To Component(LocalPose=Use cached pose 'input pose'()), Alpha=1, ExternalForce="0,0,0") {
    angularDampingOverride=0.9999
    angularSpringConstant=4000
    bAngularSpring=true
    bOverrideAngularDamping=true
    bUseGravityOverride=true
    linearDampingOverride=0.9
    linearSpringConstant=30
    boundBone {boneName=dyn_straw01_base_rot}
    componentAppliedLinearAccClamp=<900 900 900>
    componentLinearAccScale=<0.5 0.5 0.5>
    componentLinearVelScale=<0.6 0.6 0.6>
    gravityOverride {z=-980}
    physicsBodyDefinitions [{
      boundBone {boneName=dyn_straw01_base_rot}
      boxExtents=<20 20 20>
      constraintSetup {angularTargetAxis=AxisY, twistAxis=AxisY, angularLimitsMax=<50 0 0>, angularLimitsMin=<-50 0 0>, angularTarget=<0 1 0>}
      localJointOffset=<0 2.2 0>
     }]
    simSpaceSettings {maxAngularAcceleration=90, maxAngularVelocity=6}
   }
   n20 = Anim Dynamics - Bone: dyn_straw01_bend(ComponentPose=@n19, Alpha=1, ExternalForce="0,0,0") {
    angularDampingOverride=0.9999
    angularSpringConstant=3000
    bAngularSpring=true
    bOverrideAngularDamping=true
    bUseGravityOverride=true
    linearDampingOverride=0.9
    linearSpringConstant=30
    boundBone {boneName=dyn_straw01_bend}
    componentAppliedLinearAccClamp=<1000 1000 1000>
    componentLinearAccScale=<0.4 0.4 0.4>
    componentLinearVelScale=<0.5 0.5 0.5>
    gravityOverride {z=-980}
    physicsBodyDefinitions [{
      boundBone {boneName=dyn_straw01_bend}
      boxExtents=<18 18 18>
      constraintSetup {angularLimitsMax=<0 40 0>, angularLimitsMin=<0 -40 0>, angularTarget=<1 0 0>}
      localJointOffset=<1 0.3 0>
     }]
    simSpaceSettings {maxAngularAcceleration=90, maxAngularVelocity=6}
   }
   n21 = Anim Dynamics - Bone: dyn_straw02_base_rot(ComponentPose=@n20, Alpha=1, ExternalForce="0,0,0") {
    angularDampingOverride=0.999
    angularSpringConstant=4000
    bAngularSpring=true
    bOverrideAngularDamping=true
    bUseGravityOverride=true
    linearDampingOverride=0.9
    linearSpringConstant=30
    boundBone {boneName=dyn_straw02_base_rot}
    componentAppliedLinearAccClamp=<900 900 900>
    componentLinearAccScale=<0.5 0.5 0.4>
    componentLinearVelScale=<0.6 0.6 0.5>
    gravityOverride {z=-980}
    physicsBodyDefinitions [{
      boundBone {boneName=dyn_straw02_base_rot}
      boxExtents=<18 18 18>
      constraintSetup {angularTargetAxis=AxisY, twistAxis=AxisY, angularLimitsMax=<70 0 0>, angularLimitsMin=<-70 0 0>, angularTarget=<0 1 0>}
      localJointOffset=<0.2 -2 0.3>
     }]
    simSpaceSettings {maxAngularAcceleration=110, maxAngularVelocity=8}
   }
   n22 = Anim Dynamics - Bone: dyn_straw02_bend(ComponentPose=@n21, Alpha=1, ExternalForce="0,0,0") {
    angularDampingOverride=0.9999
    angularSpringConstant=3000
    bAngularSpring=true
    bOverrideAngularDamping=true
    bUseGravityOverride=true
    linearDampingOverride=0.9
    linearSpringConstant=30
    boundBone {boneName=dyn_straw02_bend}
    componentAppliedLinearAccClamp=<1000 1000 1000>
    componentLinearAccScale=<0.4 0.5 0.6>
    componentLinearVelScale=<0.5 0.6 0.6>
    gravityOverride {z=-980}
    physicsBodyDefinitions [{
      boundBone {boneName=dyn_straw02_bend}
      boxExtents=<25 10 15>
      constraintSetup {angularLimitsMax=<0 40 0>, angularLimitsMin=<0 -40 0>, angularTarget=<1 0 0>}
      localJointOffset=<0.9 0.1 -1>
     }]
    simSpaceSettings {maxAngularAcceleration=110, maxAngularVelocity=8}
   }
   n23 = Anim Dynamics - Bone: dyn_glasses(ComponentPose=@n22, Alpha=1, ExternalForce="0,0,0") {
    angularDampingOverride=0.9999
    angularSpringConstant=800
    bAngularSpring=true
    bOverrideAngularDamping=true
    bUseGravityOverride=true
    boundBone {boneName=dyn_glasses}
    componentAppliedLinearAccClamp=<1000 1000 1000>
    componentLinearAccScale=<0.5 0.5 0.5>
    componentLinearVelScale=<0.6 0.6 0.6>
    gravityOverride {z=-980}
    physicsBodyDefinitions [{
      boundBone {boneName=dyn_glasses}
      boxExtents=<20 20 20>
      constraintSetup {angularLimitsMax=<0 0 15>, angularLimitsMin=<0 0 -1>, angularTarget=<1 0 0>}
      localJointOffset=<1 0.1 0.1>
     }]
   }
   n24 = Anim Dynamics - Bone: dyn_popcorn_01(ComponentPose=@n23, bAlphaBoolEnabled=True, ExternalForce="0,0,0") {
    alphaInputType=Bool
    bLinearSpring=true
    bOverrideLinearDamping=true
    bUseGravityOverride=true
    linearDampingOverride=0.99
    linearSpringConstant=85
    alphaBoolBlend {blendOption=Linear}
    alphaScaleBiasClamp {bClampResult=true}
    boundBone {boneName=dyn_popcorn_01}
    componentAppliedLinearAccClamp=<0 0 800>
    componentLinearAccScale {z=0.4}
    componentLinearVelScale {z=0.5}
    gravityOverride {z=-980}
    physicsBodyDefinitions [{
      boundBone {boneName=dyn_popcorn_01}
      boxExtents=<10 10 10>
      constraintSetup {angularLimitsMax=<0 0 0>, angularLimitsMin=<0 0 0>, angularTarget=<0 0 0>, linearAxesMax=<1.5 0 0>}
      localJointOffset=<0.4 0.3 0>
     }]
   }
   n25 = "Transform (Modify) Bone - Bone: dyn_popcorn_02"(ComponentPose=@n24, Alpha=1, Translation="-4.797,-0.000,0.000", Rotation="0,0,0", Scale="1,1,1") {translationMode=BMM_Additive, translationSpace=BCS_BoneSpace, boneToModify {boneName=dyn_popcorn_02}, translation {x=-4.797}}
   n26 = Anim Dynamics - Bone: dyn_popcorn_02(ComponentPose=@n25, bAlphaBoolEnabled=True, ExternalForce="0,0,0") {
    alphaInputType=Bool
    bLinearSpring=true
    bOverrideLinearDamping=true
    bUseGravityOverride=true
    linearDampingOverride=0.999
    linearSpringConstant=70
    alphaBoolBlend {blendOption=Linear}
    alphaScaleBiasClamp {bClampResult=true}
    boundBone {boneName=dyn_popcorn_02}
    componentAppliedLinearAccClamp=<0 0 800>
    componentLinearAccScale {z=0.5}
    componentLinearVelScale {z=0.6}
    gravityOverride {z=-980}
    physicsBodyDefinitions [{
      boundBone {boneName=dyn_popcorn_02}
      boxExtents=<5 5 5>
      constraintSetup {angularLimitsMax=<0 0 0>, angularLimitsMin=<0 0 0>, angularTarget=<0 0 0>, linearAxesMax=<1.8 0 0>}
      localJointOffset=<0.5 0 0.1>
     }]
   }
   n27 = "Transform (Modify) Bone - Bone: dyn_popcorn_03"(ComponentPose=@n26, Alpha=1, Translation="-4.797,-0.000,0.000", Rotation="0,0,0", Scale="1,1,1") {translationMode=BMM_Additive, translationSpace=BCS_BoneSpace, boneToModify {boneName=dyn_popcorn_03}, translation {x=-4.797}}
   n28 = Anim Dynamics - Bone: dyn_popcorn_03(ComponentPose=@n27, bAlphaBoolEnabled=True, ExternalForce="0,0,0") {
    alphaInputType=Bool
    bLinearSpring=true
    bOverrideLinearDamping=true
    bUseGravityOverride=true
    linearDampingOverride=0.999
    linearSpringConstant=100
    alphaBoolBlend {blendOption=Linear}
    alphaScaleBiasClamp {bClampResult=true}
    boundBone {boneName=dyn_popcorn_03}
    componentAppliedLinearAccClamp=<0 0 750>
    componentLinearAccScale {z=0.5}
    componentLinearVelScale {z=0.6}
    gravityOverride {z=-980}
    physicsBodyDefinitions [{
      boundBone {boneName=dyn_popcorn_03}
      boxExtents=<5 5 5>
      constraintSetup {angularLimitsMax=<0 0 0>, angularLimitsMin=<0 0 0>, angularTarget=<0 0 0>, linearAxesMax=<2 0 0>}
      localJointOffset=<0.5 0 0.2>
     }]
   }
   n29 = "Transform (Modify) Bone - Bone: dyn_popcorn_04"(ComponentPose=@n28, Alpha=1, Translation="-3.548,0.000,-0.000", Rotation="0,0,0", Scale="1,1,1") {translationMode=BMM_Additive, translationSpace=BCS_BoneSpace, boneToModify {boneName=dyn_popcorn_04}, translation {x=-3.548}}
   n30 = Anim Dynamics - Bone: dyn_popcorn_04(ComponentPose=@n29, bAlphaBoolEnabled=True, ExternalForce="0,0,0") {
    alphaInputType=Bool
    bLinearSpring=true
    bOverrideLinearDamping=true
    bUseGravityOverride=true
    linearDampingOverride=0.999
    linearSpringConstant=100
    alphaBoolBlend {blendOption=Linear}
    alphaScaleBiasClamp {bClampResult=true}
    boundBone {boneName=dyn_popcorn_04}
    componentAppliedLinearAccClamp=<0 0 800>
    componentLinearAccScale {z=0.5}
    componentLinearVelScale {z=0.6}
    gravityOverride {z=-980}
    physicsBodyDefinitions [{
      boundBone {boneName=dyn_popcorn_04}
      boxExtents=<7 7 7>
      constraintSetup {angularLimitsMax=<0 0 0>, angularLimitsMin=<0 0 0>, angularTarget=<0 0 0>, linearAxesMax=<2 0 0>}
      localJointOffset=<0.5 0 0>
     }]
   }
   n31 = "Transform (Modify) Bone - Bone: dyn_popcorn_05"(ComponentPose=@n30, Alpha=1, Translation="-1.897,0.000,0.753", Rotation="0,0,0", Scale="1,1,1") {translationMode=BMM_Additive, translationSpace=BCS_BoneSpace, boneToModify {boneName=dyn_popcorn_05}, translation {x=-1.897, z=0.753}}
   n32 = Anim Dynamics - Bone: dyn_popcorn_05(ComponentPose=@n31, bAlphaBoolEnabled=True, ExternalForce="0,0,0") {
    alphaInputType=Bool
    bLinearSpring=true
    bOverrideLinearDamping=true
    bUseGravityOverride=true
    linearDampingOverride=0.999
    linearSpringConstant=80
    alphaBoolBlend {blendOption=Linear}
    alphaScaleBiasClamp {bClampResult=true}
    boundBone {boneName=dyn_popcorn_05}
    componentAppliedLinearAccClamp=<0 0 750>
    componentLinearAccScale {z=0.4}
    componentLinearVelScale {z=0.5}
    gravityOverride {z=-980}
    physicsBodyDefinitions [{
      boundBone {boneName=dyn_popcorn_05}
      boxExtents=<5 5 5>
      constraintSetup {angularLimitsMax=<0 0 0>, angularLimitsMin=<0 0 0>, angularTarget=<0 0 0>, linearAxesMax=<2 0 0>}
      localJointOffset=<0.3 0.3 0>
     }]
   }
   n33 = "Transform (Modify) Bone - Bone: dyn_popcorn_06"(ComponentPose=@n32, Alpha=1, Translation="-3.548,-1.146,-0.000", Rotation="0,0,0", Scale="1,1,1") {translationMode=BMM_Additive, translationSpace=BCS_BoneSpace, boneToModify {boneName=dyn_popcorn_06}, translation=<-3.548 -1.146>}
   n34 = Anim Dynamics - Bone: dyn_popcorn_06(ComponentPose=@n33, bAlphaBoolEnabled=True, ExternalForce="0,0,0") {
    alphaInputType=Bool
    bLinearSpring=true
    bOverrideLinearDamping=true
    bUseGravityOverride=true
    linearDampingOverride=0.99
    linearSpringConstant=100
    alphaBoolBlend {blendOption=Linear}
    alphaScaleBiasClamp {bClampResult=true}
    boundBone {boneName=dyn_popcorn_06}
    componentAppliedLinearAccClamp=<0 0 760>
    componentLinearAccScale {z=0.5}
    componentLinearVelScale {z=0.6}
    gravityOverride {z=-980}
    physicsBodyDefinitions [{
      boundBone {boneName=dyn_popcorn_06}
      boxExtents=<15 5 10>
      constraintSetup {angularLimitsMax=<0 0 0>, angularLimitsMin=<0 0 0>, angularTarget=<0 0 0>, linearAxesMax=<1.2 0 0>}
      localJointOffset=<0.5 0 0.3>
     }]
   }
   n35 = "Transform (Modify) Bone - Bone: dyn_popcorn_07"(ComponentPose=@n34, Alpha=1, Translation="-1.084,-1.146,-0.000", Rotation="0,0,0", Scale="1,1,1") {translationMode=BMM_Additive, translationSpace=BCS_BoneSpace, boneToModify {boneName=dyn_popcorn_07}, translation=<-1.084 -1.146>}
   n36 = Anim Dynamics - Bone: dyn_popcorn_07(ComponentPose=@n35, bAlphaBoolEnabled=True, ExternalForce="0,0,0") {
    alphaInputType=Bool
    bLinearSpring=true
    bOverrideLinearDamping=true
    bUseGravityOverride=true
    linearDampingOverride=0.999
    linearSpringConstant=70
    alphaBoolBlend {blendOption=Linear}
    alphaScaleBiasClamp {bClampResult=true}
    boundBone {boneName=dyn_popcorn_07}
    componentAppliedLinearAccClamp=<0 0 800>
    componentLinearAccScale {z=0.4}
    componentLinearVelScale {z=0.5}
    gravityOverride {z=-980}
    physicsBodyDefinitions [{
      boundBone {boneName=dyn_popcorn_07}
      boxExtents=<3 5 5>
      constraintSetup {angularLimitsMax=<0 0 0>, angularLimitsMin=<0 0 0>, angularTarget=<0 0 0>, linearAxesMax=<2.7 0 0>}
      localJointOffset=<0.6 0 0>
     }]
    simSpaceSettings {maxAngularAcceleration=0, maxAngularVelocity=0}
   }
   Output Pose(Result=Blend Poses by bool(BlendPose_0=Component To Local(ComponentPose=@n18), BlendPose_1=Component To Local(ComponentPose=@n36), bActiveValue=Get bIsInFrontEnd())) {layerGroup=None}
   AnimGraphNode_SubInput_0 = Input Pose()
   Save cached pose 'input pose'(Pose=@AnimGraphNode_SubInput_0) {cacheName=input pose}
  ]}
 EventGraph {nodes [
   TryGetPawnOwner()
   Event BlueprintUpdateAnimation()
  ]}
]
variables []
